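# Additive unlit particle material for Curie gameplay cues that fakes a glow and local light on surrounding surfaces.
# A radial sprite glow adds to a sphere mask around the particle at scene position, shaded by GBuffer normals and optionally tinted by stored base color.
# Particle color, depth fade, vertex-computed distance fade and two-sine flicker scale the result, with dynamic parameters driving glow, flicker and tint.
# Forward and mobile paths feed zero emissive.
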Material /Game/GameplayCueNotifies/Curie/Materials/M_Curie_FakeGlowAndLighting
graph {
 properties {
  bAutomaticallySetUsageInEditor=false
  bEnableResponsiveAA=true
  bUsedWithMeshParticles=true
  bUsedWithParticleSprites=true
  blendMode=BLEND_Additive
  pixelDepthOffsetMode=PDOM_Legacy
  shadingModel=MSM_Unlit
  usedShadingModels=Unlit
 }
 nodes [
  n1 = ParticlePositionWS()
  n2 = ParticleRadius()
  n3 = Add(A=ScalarParameter() {defaultValue=40, parameterName=FadeStartDistance}, B=@n2)
  n4 = Add(A=ScalarParameter() {defaultValue=10, parameterName=FadeEndDistance}, B=@n2)
  n5 = Multiply(A=DynamicParameter() {paramNames [GlowOpacity, FlickerSpeed, FlickerScale, BaseColorBlend]}.FlickerSpeed, B=Time()) {constA=1}
  n6 = Multiply(
   A=Add(A=Sine(None=@n5) {period=0.5651}, B=Sine(None=@n5) {period=0.442})
   B=DynamicParameter() {paramNames [GlowOpacity, FlickerSpeed, FlickerScale, BaseColorBlend]}.FlickerScale
  )
  n7 = VertexInterpolator(VS=MaterialFunctionCall(
    R=MaterialFunctionCall(Position=@n1, Fade_Start_Distance=@n3, Fade_End_Distance=@n4) {materialFunction=&MF_Distance_SyncedFade_FX}.Result
    G=Add(A=@n6)
   ) {materialFunction=&AppendMany}.RGB)
  n8 = Constant() {r=0.5}
  n9 = Power(
   Base=SphereMask(A=TextureCoordinate(), B=@n8, Radius=@n8, Hardness=Constant())
   Exp=ScalarParameter() {defaultValue=4, parameterName=GlowExponent}
  )
  n10 = DynamicParameter() {paramNames [GlowOpacity, FlickerSpeed, FlickerScale, BaseColorBlend]}
  n11 = Normalize(VectorInput=Subtract(A=WorldPosition().XYZ, B=ParticlePositionWS()))
  n12 = ComponentMask(None=SceneTexture() {sceneTextureId=PPI_WorldNormal}.Color) {b=true, g=true, r=true}
  n13 = ParticlePositionWS()
  n14 = MaterialFunctionCall() {materialFunction=&WorldPositionBehindTranslucency}
  n15 = Multiply(A=ParticleRadius(), B=ScalarParameter() {defaultValue=1, parameterName=RadiusScale}) {constB=0.725}
  n16 = Constant()
  n17 = Multiply(
   A=Saturate(None=DotProduct(A=@n11, B=@n12))
   B=Power(
    Base=SphereMask(A=@n13, B=@n14.Result, Radius=@n15, Hardness=@n16)
    Exp=ScalarParameter() {defaultValue=6, parameterName=RadiusFalloffExponent}
   )
  )
  n18 = LinearInterpolate(
   B=ComponentMask(None=SceneTexture() {sceneTextureId=PPI_StoredBaseColor}.Color) {b=true, g=true, r=true}
   Alpha=DynamicParameter() {paramNames [GlowOpacity, FlickerSpeed, FlickerScale, BaseColorBlend]}.BaseColorBlend
  ) {constA=1}
  n19 = Multiply(A=Add(A=Multiply(A=@n9, B=@n10.GlowOpacity), B=Multiply(A=@n17, B=@n18)), B=ParticleColor().RGB)
  n20 = DepthFade(FadeDistance=ScalarParameter() {defaultValue=10, parameterName=FadeDistance})
  n21 = Multiply(A=ComponentMask(None=@n7.PS) {g=true}, B=Multiply(A=ComponentMask(None=@n7.PS) {r=true}, B=Multiply(A=@n19, B=@n20)))
  n22 = Constant()
  MaterialFunctionCall(Emissive=ShadingPathSwitch(Default=@n21, Deferred=@n21, Forward=@n22, Mobile=@n22)) {materialFunction=&MF_FX_EmissiveAdjustments}
  n23 = CameraPositionWS()
  n24 = ParticlePositionWS()
  Multiply(A=Normalize(VectorInput=Subtract(A=@n23, B=@n24)), B=ParticleRadius())
  ScalarParameter() {defaultValue=0.9, parameterName=BaseColorBlend}
 ]
}
