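# Additive unlit material for the Curie propagation arc gameplay cue, drawing an animated lightning arc on a spline-thickened strip.
# Custom primitive data timestamps drive an in/out transition that scales displacement and moves a reveal mask with a bright leading edge.
# World-aligned noise displaces the arc's middle, stepped by ceiled sine-plus-cosine time for staccato jitter.
# Similarly stepped panning plasma textures supply the tinted, depth-faded glow; static switches allow plain parameters instead of primitive data.
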Material /Game/GameplayCueNotifies/Curie/Materials/M_Curie_PropagationArc
graph {
 properties {bEnableResponsiveAA=true, blendMode=BLEND_Additive, pixelDepthOffsetMode=PDOM_Legacy, shadingModel=MSM_Unlit, usedShadingModels=Unlit}
 nodes [
  n1 = StaticSwitchParameter(
   True=ScalarParameter() {bUseCustomPrimitiveData=true, defaultValue=30, parameterName=SplineWidth, primitiveDataIndex=10}
   False=ScalarParameter() {defaultValue=30, parameterName=SplineWidth}
  ) {defaultValue=true, parameterName=UseCustomPrimitveData}
  n2 = ScalarParameter() {bUseCustomPrimitiveData=true, defaultValue=1, parameterName=ArcDuration, primitiveDataIndex=2}
  n3 = MaterialFunctionCall(
   TimestampA=ScalarParameter() {bUseCustomPrimitiveData=true, defaultValue=1, parameterName=TimestampA, primitiveDataIndex=1}
   TransitionLength_In=@n2
   TimestampB=ScalarParameter() {bUseCustomPrimitiveData=true, parameterName=TimestampB, primitiveDataIndex=7}
   TransitionLength_Out=@n2
   Transition In Only?=StaticBool()
  ) {materialFunction=&MF_Timestamp_InOut_01}
  n4 = WorldPosition()
  n5 = Time()
  n6 = Time()
  n7 = Ceil(None=Add(A=Cosine(None=@n5) {period=0.6463}, B=Sine(None=@n6) {period=1.1907}))
  n8 = Multiply(A=ScalarParameter() {defaultValue=50, parameterName=WPOStaccatoScale}, B=@n7)
  n9 = MaterialFunctionCall(
   TextureObject=TextureObject() {samplerType=SAMPLERTYPE_LinearColor, texture=&T_NoisePack_LinearColor}
   TextureSize=StaticSwitchParameter(
    True=ScalarParameter() {bUseCustomPrimitiveData=true, defaultValue=920, parameterName=WPONoiseWorldTileScale, primitiveDataIndex=8}
    False=ScalarParameter() {defaultValue=920, parameterName=WPONoiseWorldTileScale}
   ) {defaultValue=true, parameterName=UseCustomPrimitveData}
   WorldPosition=Add(A=VectorParameter() {bUseCustomPrimitiveData=true, parameterName=RandomV4, primitiveDataIndex=3}.RGB, B=Add(A=@n4.XYZ, B=@n8))
  ) {materialFunction=&WorldAlignedTexture}
  n10 = ScalarParameter() {bUseCustomPrimitiveData=true, parameterName=WPONoiseScale, primitiveDataIndex=9}
  n11 = ScalarParameter() {parameterName=WPONoiseScale}
  n12 = Multiply(
   A=@n3."Linear (A Only)"
   B=Multiply(
    A=ConstantBiasScale(None=@n9.XYZ Texture) {bias=-0.5, scale=2}
    B=StaticSwitchParameter(True=@n10, False=@n11) {defaultValue=true, parameterName=UseCustomPrimitveData}
   )
  )
  n13 = Constant() {r=0.5}
  n14 = SphereMask(
   A=MaterialFunctionCall(Float2=TextureCoordinate()) {materialFunction=&BreakOutFloat2Components}.G
   B=@n13
   Radius=@n13
   Hardness=@n13
  )
  n15 = VectorParameter() {parameterName=Offset, defaultValue=<rgba 0 0 25 0>}
  n16 = MaterialFunctionCall() {materialFunction=/Engine/Functions/Engine_MaterialFunctions02/WorldPositionOffset/ObjectScale}
  Add(
   A=Add(
    A=MaterialFunctionCall(WidthBase=@n1, WidthTip=@n1) {materialFunction=&SplineThicken}.WorldPositionOffset
    B=Multiply(A=@n12, B=@n14)
   )
   B=Transform(None=Divide(A=@n15.RGB, B=@n16.Scale XYZ)) {transformSourceType=TRANSFORMSOURCE_Local}
  )
  n17 = MaterialFunctionCall(Float2=TextureCoordinate()) {materialFunction=&BreakOutFloat2Components}
  n18 = Time()
  n19 = StaticSwitchParameter(
   True=VertexInterpolator(VS=AppendVector(A=@n3."Linear (A Only)", B=@n3."Linear (A Only)")).PS
   False=Frac(None=Multiply(A=@n18) {constB=0.25})
  ) {defaultValue=true, parameterName=UseCustomPrimitveData}
  n20 = ConstantBiasScale(None=ComponentMask(None=@n19) {r=true}) {bias=-0.25, scale=2}
  n21 = SphereMask(A=@n17.G, B=Constant(), Radius=@n20, Hardness=Constant() {r=0.8})
  n22 = TextureCoordinate()
  n23 = Time()
  n24 = Time()
  n25 = Ceil(None=Add(A=Cosine(None=@n23) {period=0.7271}, B=Sine(None=@n24) {period=1.1016}))
  n26 = AppendVector(A=Constant(), B=ScalarParameter() {defaultValue=1.33, parameterName=PanU1Speed})
  n27 = AppendVector(
   A=Constant() {r=1}
   B=StaticSwitchParameter(
    True=ScalarParameter() {bUseCustomPrimitiveData=true, defaultValue=1, parameterName=VTiling}
    False=ScalarParameter() {defaultValue=1, parameterName=VTiling}
   ) {defaultValue=true, parameterName=UseCustomPrimitveData}
  )
  n28 = TextureSample(UVs=Multiply(A=Panner(Coordinate=@n22, Time=@n25, Speed=@n26) {bFractionalPart=true}, B=@n27)) {texture=/Game/Effects/Fort_Effects/Textures/Tile/T_Lightning_Plasma_Soft}
  n29 = Panner(
   Coordinate=Multiply(A=TextureCoordinate() {vTiling=0.859}, B=Constant2Vector() {g=1, r=-1})
   Time=@n7
   Speed=AppendVector(A=Constant(), B=ScalarParameter() {defaultValue=1.66, parameterName=PanU2Speed})
  ) {bFractionalPart=true}
  n30 = Add(A=@n28.R, B=TextureSample(UVs=Multiply(A=@n29, B=@n27)) {texture=&T_Lightning_Plasma_Soft}.G)
  n31 = ComponentMask(None=@n19) {g=true}
  n32 = VectorParameter() {parameterName=ColorLow, defaultValue=<rgba 1.7992 1.7992 40 0>}
  n33 = VectorParameter() {parameterName=ColorHigh, defaultValue=<rgba 4.1667 4.1667 200 0>}
  n34 = SphereMask(
   A=@n17.G
   B=Subtract(A=@n20) {constB=0.2}
   Radius=ScalarParameter() {defaultValue=0.4, parameterName=LeadingArcRadius}
   Hardness=Constant() {r=0.7}
  )
  n35 = ScalarParameter() {defaultValue=50, parameterName=LeadingArcIntensity}
  n36 = Add(A=Multiply(A=Multiply(A=@n34, B=@n35) {constB=100}, B=@n28.R), B=@n28.R)
  n37 = ScalarParameter() {bUseCustomPrimitiveData=true, defaultValue=30, parameterName=ColorScale, primitiveDataIndex=11}
  n38 = Constant() {r=1}
  n39 = Multiply(
   A=Subtract(A=Divide(A=@n30), B=OneMinus(None=@n31))
   B=Multiply(
    A=LinearInterpolate(A=@n32.RGB, B=@n33.RGB, Alpha=@n36)
    B=StaticSwitchParameter(True=@n37, False=@n38) {defaultValue=true, parameterName=UseCustomPrimitveData}
   )
  )
  DepthFade(Opacity=Multiply(A=Multiply(A=@n21, B=@n39), B=@n14), FadeDistance=ScalarParameter() {defaultValue=15, parameterName=DepthFadeDistance})
 ]
}
